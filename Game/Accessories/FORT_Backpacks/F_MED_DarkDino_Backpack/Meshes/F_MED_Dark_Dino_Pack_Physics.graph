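# Physics asset for the Dark Dino backpack accessory, giving its three dynamic "hair" bones simulated secondary motion.
# Small capsule bodies cover the left, middle and right hair bones plus a smaller one on the pack root bone.
# Each hair body is constrained directly to the pack rather than chained, with moderate swing limits and tight twist, so the pieces sway independently with little roll.
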PhysicsAsset /Game/Accessories/FORT_Backpacks/F_MED_DarkDino_Backpack/Meshes/F_MED_Dark_Dino_Pack_Physics
bodies {
 dyn_hair_l [{shapeType=Capsule, length=4, radius=4, center=<1 0 0>, rotation=<rot 90 0 0>}]
 dyn_hair_mid [{shapeType=Capsule, length=4, radius=4, center=<1.5 0 0>, rotation=<rot 90 0 0>}]
 dyn_hair_r [{shapeType=Capsule, length=4, radius=4, center=<1 0 0>, rotation=<rot 90 0 0>}]
 pack [{shapeType=Capsule, length=2, radius=2}]
}
constraints [
 {bone1Name=dyn_hair_mid, bone2Name=pack, swing1LimitDegrees=30, swing2LimitDegrees=30, twistLimitDegrees=5}
 {bone1Name=dyn_hair_l, bone2Name=pack, swing1LimitDegrees=30, swing2LimitDegrees=30, twistLimitDegrees=5}
 {bone1Name=dyn_hair_r, bone2Name=pack, swing1LimitDegrees=30, swing2LimitDegrees=30, twistLimitDegrees=5}
]
